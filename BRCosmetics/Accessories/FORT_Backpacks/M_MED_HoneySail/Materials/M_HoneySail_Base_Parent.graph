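# Translucent, unlit, two-sided parent material for the HoneySail back bling, drawing animated smoky butterfly-style wings.
# Two panning chromatic smoke samples are tinted by CrackColor and added to a sphere-masked base-to-crack gradient, with Fresnel-boosted color.
# The same smoke, multiplied by a butterfly mask and Opacity, drives transparency.
# MF_Butterfly_WPO animates vertices behind UseVertShader; the Hide Entirely switch instead collapses the mesh and swaps opacity for a constant.
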Material /BRCosmetics/Accessories/FORT_Backpacks/M_MED_HoneySail/Materials/M_HoneySail_Base_Parent
graph {
 properties {
  bUsedWithInstancedStaticMeshes=true
  bUsedWithMeshDeformer=true
  bUsedWithSkeletalMesh=true
  blendMode=BLEND_Translucent
  pixelDepthOffsetMode=PDOM_Legacy
  refractionMethod=RM_IndexOfRefraction
  shadingModel=MSM_Unlit
  twoSided=true
  usedShadingModels=Unlit
 }
 nodes [
  n1 = MaterialFunctionCall!(TimeOffset=Constant() {r=-0.015}, WingBendMult=Constant() {r=1.1}) {materialFunction=/Game/Athena/Prototype/Blueprints/Island/MF_Butterfly_WPO}
  Add(
   A=@n1.WPO
   B=MaterialFunctionCall(Offset Amount=ScalarParameter() {defaultValue=-20, parameterName=CameraOffset}) {materialFunction=/Engine/Functions/Engine_MaterialFunctions02/WorldPositionOffset/CameraOffset}.Camera Offset
  )
  Constant() {r=1}
  n2 = TextureCoordinate() {uTiling=3, vTiling=1.4}
  n3 = ScalarParameter() {defaultValue=1, parameterName=Opacity}
  n4 = Multiply(A=Subtract(A=Time(), B=Multiply(A=@n3) {constB=2}), B=ScalarParameter() {defaultValue=2, parameterName=PanSpeed})
  n5 = TextureCoordinate() {uTiling=1.5, vTiling=0.7}
  n6 = Add(
   A=TextureSample(UVs=Panner(Coordinate=@n2, Time=@n4) {speedX=-0.04, speedY=0.15}) {texture=&T_Inky_Smoke_Tile_Chromatic}.RGB
   B=TextureSample(UVs=Panner(Coordinate=@n5, Time=@n4) {speedX=0.05, speedY=0.25}) {texture=&T_Inky_Smoke_Tile_Chromatic}.RGB
  )
  n7 = VectorParameter() {parameterName=CrackColor, defaultValue=<rgba 0.1905 0.4686 0.88 0>}
  n8 = SphereMask(A=TextureCoordinate() {coordinateIndex=1}, B=Constant2Vector() {g=0.5, r=1}) {attenuationRadius=0.5, hardnessPercent=0}
  n9 = Add(
   A=Multiply(A=@n6, B=Multiply(A=@n7.RGB) {constB=1.5})
   B=LinearInterpolate(A=VectorParameter() {parameterName=BaseColor, defaultValue=<rgba 0.1137 0.1382 0.37 0>}.RGB, B=@n7.RGB, Alpha=Power(Base=@n8))
  )
  MaterialFunctionCall(Emissive=Desaturation(None=@n9, Fraction=ScalarParameter() {parameterName=Desaturation})) {materialFunction=&MF_FX_EmissiveAdjustments}
  MaterialFunctionCall() {materialFunction=&DebugTimeSine}
  Multiply(A=LinearInterpolate(Alpha=Fresnel() {baseReflectFraction=0, exponent=3}) {constA=0.7, constB=2}, B=@n9)
  n10 = LocalPosition()
  StaticSwitchParameter(
   True=Transform(None=Multiply(A=@n10.XYZ) {constB=-1}) {transformSourceType=TRANSFORMSOURCE_Local}
   False=StaticSwitchParameter(True=@n1.WPO, False=Constant()) {defaultValue=true, parameterName=UseVertShader}
  ) {parameterName=Hide Entirely}
  n11 = Add(A=Multiply(A=@n6) {constB=0.75}) {constB=0.7}
  n12 = TextureSample(UVs=TextureCoordinate() {coordinateIndex=1, uTiling=2}) {texture=/Game/Athena/Prototype/Blueprints/Island/T_Butterfly_M}
  StaticSwitchParameter(True=Constant(), False=Multiply(A=Multiply(A=@n11, B=@n12.R), B=@n3)) {parameterName=Hide Entirely}
 ]
}
